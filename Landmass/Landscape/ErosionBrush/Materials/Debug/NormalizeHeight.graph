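# Unlit alpha-composite debug material for the Landmass erosion brush that remaps height data from a render target into a 0–1 range.
# It samples the blue channel of the RT texture parameter at an explicit mip level, scales it, normalizes it between the Min and Max parameters and saturates the result.
# A half-texel UV offset driven by resolution parameter R is present but not wired into the sample.
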Material /Landmass/Landscape/ErosionBrush/Materials/Debug/NormalizeHeight
graph {
 properties {blendMode=BLEND_AlphaComposite, pixelDepthOffsetMode=PDOM_Legacy, shadingModel=MSM_Unlit, usedShadingModels=Unlit}
 nodes [
  Add(A=TextureCoordinate(), B=Divide(B=ScalarParameter() {defaultValue=2048, parameterName=R}) {constA=0.5})
  Constant()
  n1 = TextureSampleParameter2D!(UVs, Level=Constant(), Apply View MipBias) {
   mipValueMode=TMVM_MipLevel
   parameterName=RT
   samplerType=SAMPLERTYPE_LinearColor
   texture=/Engine/Functions/Engine_MaterialFunctions02/PivotPainter2/Black_1x1_EXR_Texture
  }
  n2 = Multiply(A=@n1.B) {constB=100}
  n3 = ScalarParameter() {defaultValue=1, parameterName=Min}
  n4 = ScalarParameter() {defaultValue=5001, parameterName=Max}
  Saturate(None=Divide(A=Subtract(A=@n2, B=@n3), B=Subtract(A=@n4, B=@n3)))
 ]
}
